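# Niagara system for the Lake vehicle boost cosmetic, a trail emitted from the car's boost sockets.
# A local-space mesh and sprite emitter forms the nozzle effect, while another sprite and mesh emitter swirls particles backward on time-rotated shape locations.
# Two per-socket ribbon emitters add trails with sine waves that grow with particle age.
# ColorVariantID drives desaturation and gain, painted-variant overrides supply colors, and GameModeID switches rates, lifetimes and motion.
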NiagaraSystem /VehicleCosmetics/Boosts/Boost_Lake/FX/NS_Boost_Lake01
dependencies {
 usedDataInterfaces [
  &NiagaraDataInterfaceArrayFloat
  &NiagaraDataInterfaceCurve
  &NiagaraDataInterfaceColorCurve
  &NiagaraDataInterfaceSpriteRendererInfo
  &NiagaraDataInterfaceSkeletalMesh
 ]
 usedDynamicInputs [&MakeFloatFromVector, &VectorFromFloat]
 usedModules [
  &self:SystemSpawnScript.SystemScriptSource.SystemScriptGraph.NiagaraNodeAssignment_0.SetVariables_#1
  &SystemState
 ]
}
emitters {
 Meshes-LocalSpace {
  emitterSpawnScript [/SetVariables(
    Emitter.Desaturation:NiagaraFloat=&SelectFloatFromArray(Direct Array Index=$System.ColorVariantID, Float Selection Array=DataInterface() {propertyValues {FloatData [0,1,1,1,1,1,1,1,1,1,1,1,1,1]}})
    Emitter.Gain:NiagaraFloat=&SelectFloatFromArray(
     Direct Array Index=$System.ColorVariantID
     Float Selection Array=DataInterface() {propertyValues {FloatData [1,1,8,5,5,3,3,2.5,3.5,3.5,2,2,3,2]}}
    )
    Emitter.NeedReflection:NiagaraFloat=&SelectFloatFromArray(
     Direct Array Index=$System.ColorVariantID
     Float Selection Array=DataInterface() {propertyValues {FloatData [0,0.5,0.6,0.5,0.5,0.5,0.5,0.5,0.5,0.6,0.6,0.6,0.5,0.5]}}
    )
   ) {isSetParameters=true}]
  emitterUpdateScript [
   &EmitterState(Inactive Response=::NewEnumerator2, Life Cycle Mode=::NewEnumerator1, Loop Duration=1, UseLoopDelay=false)
   &VC_FilteredSocketCount()
   &SpawnRate(SpawnRate=18, Use Spawn Probability=false)
   &SpawnRate(SpawnGroup=1, SpawnRate=18, Use Spawn Probability=false) {module=SpawnRate003}
   &SpawnRate(SpawnGroup=2, SpawnRate=22, Use Spawn Probability=true) {module=SpawnRate004}
  ]
  particleSpawnScript [
   /SetVariables(Particles.VisTag:NiagaraInt32=&SpawnGroup()) {isSetParameters=true}
   &VC_PaintedVariantsOverride(
    BurntSienna=<rgba 0.0642 0.0198 0.0079 1>
    DefaultColor=<rgba 0 2 1.1728 0.09>
    R1 Black:LinearColor=<rgba 0.0208 0.0208 0.0208 0.8>
    R1 BurntSienna:LinearColor=<rgba 0.0573 0.0172 0 1>
    R1 Cobalt:LinearColor=<rgba 0.014 0.0298 0.8958 1>
    R1 Crimson:LinearColor=<rgba 0.6172 0.0252 0.0252 1>
    R1 ForestGreen:LinearColor=<rgba 0 0.4219 0.0213 1>
    R1 Grey:LinearColor=<rgba 0.5521 0.5521 0.5521 1>
    R1 Lime:LinearColor=<rgba 0.5717 1 0.0008 1>
    R1 Orange:LinearColor=<rgba 1.4167 0.1533 0 0.687>
    R1 Pink:LinearColor=<rgba 1.5208 0.4198 1.1205 1>
    R1 Purple:LinearColor=<rgba 0.6092 0.1719 1.5 1>
    R1 Saffron:LinearColor=<rgba 1.3 1.0125 0.0677 1>
    R1 SkyBlue:LinearColor=<rgba 0 0.9964 1.5 1>
    R1 TitaniumWhite:LinearColor=<rgba 1.5 1.5 1.5 1>
    R3 Black:LinearColor=<rgba 0.03 0.03 0.03 2.5>
    R3 BurntSienna:LinearColor=<rgba 0.0625 0.0193 0.0077 2>
    R3 Cobalt:LinearColor=<rgba 0.0308 0.0744 0.7396 2>
    R3 Crimson:LinearColor=<rgba 0.7188 0.015 0.034 5>
    R3 DefaultColor:LinearColor=<rgba 0.15 1 0.56 1.5>
    R3 ForestGreen:LinearColor=<rgba 0 0.4167 0.0211 2>
    R3 Grey:LinearColor=<rgba 0.2708 0.2708 0.2708 1>
    R3 Lime:LinearColor=<rgba 0.3698 0.7396 0 2>
    R3 Orange:LinearColor=<rgba 0.8906 0.2778 0 1>
    R3 Pink:LinearColor=<rgba 0.7344 0.2142 0.5452 2>
    R3 Purple:LinearColor=<rgba 0.2831 0.1195 0.552 1>
    R3 Saffron:LinearColor=<rgba 0.85 0.6517 0 1>
    R3 SkyBlue:LinearColor=<rgba 0 0.4471 0.6719 1>
    R3 TitaniumWhite:LinearColor=<rgba 0.6719 0.6719 0.6719 1>
    Use R1 Painted Overrides=true
    Use R3 Painted Overrides=true
   )
   &InitializeParticle(
    Lifetime=&MaskFloatBySpawnGroup(
     Masked Float=&RandomRangeFloat(Minimum=0.5, Maximum=0.7)
     Passthrough Float=&MaskFloatBySpawnGroup(
      Masked Float=&RandomRangeFloat(Minimum=0.25, Maximum=0.5)
      Passthrough Float=&MaskFloatBySpawnGroup(Masked Float=&RandomRangeFloat(Minimum=0.25, Maximum=0.45), Passthrough Float=0, Spawn Group Mask=1)
     )
     Spawn Group Mask=2
    )
    Mesh Scale Mode=::NewEnumerator3
    Mesh Uniform Scale=&MaskFloatBySpawnGroup(
     Masked Float=0.14
     Passthrough Float=&MaskFloatBySpawnGroup(Masked Float=0.14, Passthrough Float=0)
     Spawn Group Mask=1
    )
    Sprite Rotation Angle=&RandomRangeFloat(Minimum=-30, Maximum=30)
    Sprite Rotation Mode=::NewEnumerator0
    Sprite Size Max=<25 50>
    Sprite Size Min=<20 45>
    Sprite Size Mode=::NewEnumerator2
    UsePositionOffset=false
    Write Material Random=true
   )
   &PartitionParticles(Number Of Partitions=$Emitter.PortCount)
   &SkeletalMeshLocation(
    Bone / Socket Index=$Particles.PartitionParticles.Partition
    Bone / Socket Sampling Mode=::NewEnumerator5
    MaskSpawnGroup=false
    Mesh Sampling Type=::NewEnumerator6
    Sampled Position Offset=&AddVectorToPosition(
     Position=$User.DTCarOffset
     Vector=&MaskVectorBySpawnGroup(
      Masked Vector=<-62 0 0>
      Passthrough Vector=&MaskVectorBySpawnGroup(
       Masked Vector=<-48 0 0>
       Passthrough Vector=&MaskVectorBySpawnGroup(Masked Vector=<-55 0 0>, Passthrough Vector=<0 0 0>, Spawn Group Mask=1)
      )
      Spawn Group Mask=2
     )
    )
    Skeletal Mesh=$System.SkeletalMesh
   )
   &AddVelocity(
    Cone Angle=&RandomRangeFloat(Minimum=-120, Maximum=120)
    Inner Cone Angle:NiagaraFloat=&RandomRangeFloat(Minimum=-90, Maximum=90)
    Module.UseConeFalloff=false
    Module.UseVelDistribution=false
    Override Seed=false
    Velocity Mode=::NewEnumerator2
    Velocity Speed=&MaskFloatBySpawnGroup(
     Masked Float=-110
     Passthrough Float=&MaskFloatBySpawnGroup(
      Masked Float=0
      Passthrough Float=&MaskFloatBySpawnGroup(Masked Float=-25, Passthrough Float=1, Spawn Group Mask=1)
     )
     Spawn Group Mask=2
    )
   )
   /SetVariables(Particles.Sprite:LinearColor=<rgba 0.15 1 0.56 1>, Particles.Waves:LinearColor=<rgba 0.0021 0.8646 0.7143 0.2>) {isSetParameters=true}
  ]
  particleUpdateScript [
   &ParticleState()
   &ScaleSpriteSize(
    Uniform Curve Scale=1.5
    Uniform Curve Sprite Scale=DataInterface() {propertyValues {Curve {keys [<at 0 0.6 RCIM_Cubic>,<at 1 1.6 RCIM_Cubic>]}}}
   )
   &UpdateMeshOrientation(Rotation Rate=&RandomRangeFloat(Minimum=-1, Maximum=1))
   &ScaleMeshSize(Scale Factor=&Multiply_VectorByFloat(
     Vector=<1 1 1>
     Float=&MaskFloatBySpawnGroup(
      Masked Float=&FloatFromCurve(FloatCurve=DataInterface() {propertyValues {Curve {keys [<at 0 1.1 RCIM_Cubic>,<at 1 0.75 RCIM_Cubic>]}}})
      Passthrough Float=&MaskFloatBySpawnGroup(
       Masked Float=&FloatFromCurve(FloatCurve=DataInterface() {propertyValues {Curve {keys [<at 0 1 RCIM_Cubic>,<at 1 0.2 RCIM_Cubic>]}}})
       Passthrough Float=0
       Spawn Group Mask=1
      )
     )
    ))
   &SolveForcesAndVelocity(Clamp Velocity=false, Limit Acceleration=false)
   &SubUVAnimation(
    "Sprite Renderer "=DataInterface() {propertyValues {SpriteRenderer=&self:Meshes-LocalSpace_0.NiagaraSpriteRendererProperties_0}}
    UseEndFrame=false
    UseStartFrame=false
   )
  ]
  rendererClasses [&NiagaraSpriteRendererProperties, &NiagaraMeshRendererProperties]
  renderers [
   {rendererClass=&NiagaraSpriteRendererProperties}
   {rendererClass=&NiagaraMeshRendererProperties, rendererIndex=1}
   {rendererClass=&NiagaraMeshRendererProperties, rendererIndex=2}
   {rendererClass=&NiagaraSpriteRendererProperties, rendererIndex=3}
   {rendererClass=&NiagaraSpriteRendererProperties, rendererIndex=4}
  ]
 }
 NE_VC_Boost_Ribbon_Parent {
  emitterSpawnScript [/SetVariables(Emitter.Desaturation:NiagaraFloat=&SelectFloatFromArray(
     Direct Array Index=$System.ColorVariantID
     Float Selection Array=DataInterface() {propertyValues {FloatData [0.2,1,1,1,1,1,1,1,1,1,1,1,1,1]}}
    )) {isSetParameters=true, module=SetVariables_#2}]
  emitterUpdateScript [
   &EmitterState(Life Cycle Mode=::NewEnumerator1, Loop Duration=1, UseLoopDelay=false)
   &VC_FilteredSocketCount()
   &VC_RateQualityHandler(Epic Spawn Rate=30, High Spawn Rate=30, Low Spawn Rate=30, Use Custom Rate?=true)
   &VC_SpawnRibbon_v1(Number of Ribbons=$Emitter.PortCount, SpawnRate=$Emitter.Spawn Rate)
  ]
  particleSpawnScript [
   &VC_InitializeRibbon_v1(
    Custom Offset=true
    Enable Skeletal Position=true
    Lifetime=&MergeHelper_FloatSwitch(0=1, 1=1, 2=1, 3=?, Index=$System.GameModeID, Num=::NewEnumerator1)
    Local Vector Offset=&AddVectorToPosition(Position=<-55 0 0>, Vector=Linked() {linkedVariable {name=User.DTCarOffset, type=NiagaraPosition}})
    Offset Amount=1
    Ribbon Width=42
   )
   &AddVelocity(Velocity=&MergeHelper_VectorSwitch(0=<-1200 0 0>, 1=<-1000 0 0>, 2=<-1500 0 0>, 3=?, Index=$System.GameModeID, Num=::NewEnumerator1)) {module=AddVelocity001}
  ]
  particleUpdateScript [
   &VC_PaintedVariantsOverride(
    DefaultColor=&ColorFromCurve(ColorCurve=DataInterface() {propertyValues {
       AlphaCurve {keys [<at 0 1 RCIM_Cubic>,<at 0.3828 1>,<at 0.5954 1>,<at 0.9983 1 RCIM_Cubic>]}
       BlueCurve {keys [<at 0 0.9645>,<at 0.167 0.9936>,<at 0.3563 2.5>,<at 0.8373 1.3319>]}
       GreenCurve {keys [<at 0 0.9489>,<at 0.167 0.9991>,<at 0.3563 0.5635>,<at 0.8373 0.5225>]}
       RedCurve {keys [<at 0 1>,<at 0.167 1>,<at 0.3563 2.1256>,<at 0.8373 5>]}
      }})
    R1 Black:LinearColor=&ColorFromCurve(ColorCurve=DataInterface() {propertyValues {
       AlphaCurve {keys [<at 0 5 RCIM_Cubic>,<at 0.2633 5>,<at 0.3828 1>,<at 0.5954 1>,<at 0.9983 1 RCIM_Cubic>]}
       BlueCurve {keys [<at 0 0.0625>,<at 0.5142 0.0365>,<at 1 0.1095>]}
       GreenCurve {keys [<at 0 0.0625>,<at 0.5142 0.0365>,<at 1 0.1095>]}
       RedCurve {keys [<at 0 0.0625>,<at 0.5142 0.0365>,<at 1 0.1095>]}
      }})
    R1 BurntSienna:LinearColor=&ColorFromCurve(ColorCurve=DataInterface() {propertyValues {
       AlphaCurve {keys [<at 0 1 RCIM_Cubic>,<at 0.3828 1>,<at 0.5954 1>,<at 0.9983 1 RCIM_Cubic>]}
       BlueCurve {keys [<at 0 0>,<at 0.2511 0>,<at 0.4573 0>,<at 0.8373 0>]}
       GreenCurve {keys [<at 0 0.0423>,<at 0.2511 0.0423>,<at 0.4573 0.0423>,<at 0.8373 0.0477>]}
       RedCurve {keys [<at 0 0.1248>,<at 0.2511 0.1248>,<at 0.4573 0.1248>,<at 0.8373 0.1406>]}
      }})
    R1 Cobalt:LinearColor=&ColorFromCurve(ColorCurve=DataInterface() {propertyValues {
       AlphaCurve {keys [<at 0 1 RCIM_Cubic>,<at 0.2633 1>,<at 0.3828 1>,<at 0.5954 1>,<at 0.9983 1 RCIM_Cubic>]}
       BlueCurve {keys [<at 0 1>,<at 0.167 1>,<at 0.3563 1>,<at 0.8373 1>]}
       GreenCurve {keys [<at 0 0.0594>,<at 0.167 0.0594>,<at 0.3563 0.0492>,<at 0.8373 0.085>]}
       RedCurve {keys [<at 0 0.0417>,<at 0.167 0.0417>,<at 0.3563 0.0312>,<at 0.8373 0.0677>]}
      }})
    R1 Crimson:LinearColor=&ColorFromCurve(ColorCurve=DataInterface() {propertyValues {
       AlphaCurve {keys [<at 0 1 RCIM_Cubic>,<at 0.2633 1>,<at 0.3828 1>,<at 0.5954 1>,<at 0.9983 1 RCIM_Cubic>]}
       BlueCurve {keys [<at 0 0.0252>,<at 0.8373 0.0423>]}
       GreenCurve {keys [<at 0 0.0252>,<at 0.8373 0.0285>]}
       RedCurve {keys [<at 0 0.6172>,<at 0.8373 0.5469>]}
      }})
    R1 DefaultColor=&ColorFromCurve(ColorCurve=DataInterface() {propertyValues {
       AlphaCurve {keys [<at 0 1 RCIM_Cubic>,<at 0.3828 1>,<at 0.5954 1>,<at 0.9983 1 RCIM_Cubic>]}
       BlueCurve {keys [<at 0 0.7716>,<at 0.2511 0.6955>,<at 0.4573 0.9038>,<at 0.8373 0.9742>]}
       GreenCurve {keys [<at 0 0.7591>,<at 0.2511 0.6994>,<at 0.4573 0.4795>,<at 0.8373 0.5197>]}
       RedCurve {keys [<at 0 0.8>,<at 0.2511 0.7>,<at 0.4573 2>,<at 0.8373 5>]}
      }})
    R1 ForestGreen:LinearColor=&ColorFromCurve(ColorCurve=DataInterface() {propertyValues {
       AlphaCurve {keys [<at 0 1 RCIM_Cubic>,<at 0.2633 1>,<at 0.3828 1>,<at 0.5954 1>,<at 0.9983 1 RCIM_Cubic>]}
       BlueCurve {keys [<at 0 0.0206>,<at 0.167 0.0214>,<at 0.3563 0.0232>,<at 0.8373 0.0815>]}
       GreenCurve {keys [<at 0 0.4115>,<at 0.167 0.4271>,<at 0.3563 0.4635>,<at 0.8373 0.4688>]}
       RedCurve {keys [<at 0 0>,<at 0.167 0>,<at 0.3563 0>,<at 0.8373 0.061>]}
      }})
    R1 Grey:LinearColor=&ColorFromCurve(ColorCurve=DataInterface() {propertyValues {
       AlphaCurve {keys [<at 0 1 RCIM_Cubic>,<at 0.2633 1>,<at 0.3828 1>,<at 0.5954 1>,<at 0.9983 1 RCIM_Cubic>]}
       BlueCurve {keys [<at 0 0.2384>,<at 0.167 0.2384>,<at 0.3563 0.3021>,<at 0.8373 0.3021>]}
       GreenCurve {keys [<at 0 0.2384>,<at 0.167 0.2384>,<at 0.3563 0.3021>,<at 0.8373 0.3021>]}
       RedCurve {keys [<at 0 0.2384>,<at 0.167 0.2384>,<at 0.3563 0.3021>,<at 0.8373 0.3021>]}
      }})
    R1 Lime:LinearColor=&ColorFromCurve(ColorCurve=DataInterface() {propertyValues {
       AlphaCurve {keys [<at 0 1 RCIM_Cubic>,<at 0.2633 1>,<at 0.3828 1>,<at 0.5954 1>,<at 0.9983 1 RCIM_Cubic>]}
       BlueCurve {keys [<at 0 0>,<at 0.167 0>,<at 0.3563 0>,<at 0.8373 0.0865>]}
       GreenCurve {keys [<at 0 0.875>,<at 0.167 0.7708>,<at 0.3563 0.6875>,<at 0.8373 0.7552>]}
       RedCurve {keys [<at 0 0.4387>,<at 0.167 0.3865>,<at 0.3563 0.3447>,<at 0.8373 0.4218>]}
      }})
    R1 Orange:LinearColor=&ColorFromCurve(ColorCurve=DataInterface() {propertyValues {
       AlphaCurve {keys [<at 0 1 RCIM_Cubic>,<at 0.2633 1>,<at 0.3828 1>,<at 0.5954 1>,<at 0.9983 1 RCIM_Cubic>]}
       BlueCurve {keys [<at 0.3563 0>,<at 0.8373 0.0004>]}
       GreenCurve {keys [<at 0.3563 0.1122>,<at 0.8373 0.1068>]}
       RedCurve {keys [<at 0.3563 0.8632>,<at 0.8373 0.8632>]}
      }})
    R1 Pink:LinearColor=&ColorFromCurve(ColorCurve=DataInterface() {propertyValues {
       AlphaCurve {keys [<at 0 1 RCIM_Cubic>,<at 0.2633 1>,<at 0.3828 1>,<at 0.5954 1>,<at 0.9983 1 RCIM_Cubic>]}
       BlueCurve {keys [<at 0 0.5096>,<at 0.167 0.5089>,<at 0.3563 0.5089>,<at 0.8693 0.5089>]}
       GreenCurve {keys [<at 0 0.2113>,<at 0.167 0.2122>,<at 0.3563 0.2122>,<at 0.8693 0.2122>]}
       RedCurve {keys [<at 0 0.5729>,<at 0.167 0.5711>,<at 0.3563 0.5711>,<at 0.8693 0.5711>]}
      }})
    R1 Purple:LinearColor=&ColorFromCurve(ColorCurve=DataInterface() {propertyValues {
       AlphaCurve {keys [<at 0 1 RCIM_Cubic>,<at 0.2633 1>,<at 0.3828 1>,<at 0.5954 1>,<at 0.9983 1 RCIM_Cubic>]}
       BlueCurve {keys [<at 0 0.4233>,<at 0.3563 0.4233>,<at 0.8693 0.7292>]}
       GreenCurve {keys [<at 0 0.0666>,<at 0.3563 0.0666>,<at 0.8693 0.1165>]}
       RedCurve {keys [<at 0 0.2122>,<at 0.3563 0.2122>,<at 0.8693 0.3692>]}
      }})
    R1 Saffron:LinearColor=&ColorFromCurve(ColorCurve=DataInterface() {propertyValues {
       AlphaCurve {keys [<at 0 1 RCIM_Cubic>,<at 0.2633 1>,<at 0.3828 1>,<at 0.5954 1>,<at 0.9983 1 RCIM_Cubic>]}
       BlueCurve {keys [<at 0 0>,<at 0.167 0>,<at 0.3563 0.0386>,<at 0.8693 0.0631>]}
       GreenCurve {keys [<at 0 0.5986>,<at 0.167 0.5667>,<at 0.3563 0.7194>,<at 0.8693 0.7291>]}
       RedCurve {keys [<at 0 0.7812>,<at 0.167 0.7396>,<at 0.3563 0.9271>,<at 0.8693 0.9323>]}
      }})
    R1 SkyBlue:LinearColor=&ColorFromCurve(ColorCurve=DataInterface() {propertyValues {
       AlphaCurve {keys [<at 0 1 RCIM_Cubic>,<at 0.2633 1>,<at 0.3828 1>,<at 0.5954 1>,<at 0.9983 1 RCIM_Cubic>]}
       BlueCurve {keys [<at 0 1>,<at 0.167 1>,<at 0.3563 1>,<at 0.8693 0.9047>]}
       GreenCurve {keys [<at 0 0.6683>,<at 0.167 0.6675>,<at 0.3563 0.6675>,<at 0.8693 0.6352>]}
       RedCurve {keys [<at 0 0>,<at 0.167 0>,<at 0.3563 0>,<at 0.8693 0.0942>]}
      }})
    Use R1 Painted Overrides=true
   )
   &ParticleState()
   &ScaleRibbonWidth(Ribbon Width Scale=&FloatFromCurve(FloatCurve=DataInterface() {propertyValues {Curve {keys [
         {arriveTangent=5.3557, interpMode=RCIM_Cubic, leaveTangent=5.3557, tangentMode=RCTM_User, time=0, value=0.9}
         {arriveTangent=-2.9288, interpMode=RCIM_Cubic, leaveTangent=-2.9288, tangentMode=RCTM_User, time=0.5, value=1.1}
         {arriveTangent=-0.0948, interpMode=RCIM_Cubic, leaveTangent=-0.0948, tangentMode=RCTM_User, time=0.95, value=0}
        ]}}}))
   &InheritSourceMovement(ApplyForce=false, ApplyPosition=true, ApplyPositionFalloff=false)
   &SolveForcesAndVelocity(Clamp Velocity=false, Limit Acceleration=false)
   &Color(Color=&ColorFromCurve(ColorCurve=DataInterface() {propertyValues {
       AlphaCurve {keys [<at 0 1 RCIM_Cubic>,<at 0.3828 1>,<at 0.5954 1>,<at 0.9983 1 RCIM_Cubic>]}
       BlueCurve {keys [<at 0 0.7716>,<at 0.2511 0.6955>,<at 0.4573 0.9038>,<at 0.8373 0.9742>]}
       GreenCurve {keys [<at 0 0.7591>,<at 0.2511 0.6994>,<at 0.4573 0.4795>,<at 0.8373 0.5197>]}
       RedCurve {keys [<at 0 0.8>,<at 0.2511 0.7>,<at 0.4573 2>,<at 0.8373 5>]}
      }}))
   &VC_UpdateRibbon_v1(Enable Skeletal Position=true)
   /SetVariables(Particles.Position=&AddVectorToPosition(
     Position=$Particles.Position
     Vector=&TransformVector(
      Vector=&MakeVector(
       X=0
       Y=&Sine(
        Normalized Angle=&Add_Float(
         A=&Multiply_Float(A=$Particles.NormalizedAge, B=-1)
         B=&Multiply_Float(A=$Emitter.LoopedAge, B=1)
        )
        Period=1
        Scale=&FloatFromCurve(FloatCurve=DataInterface() {propertyValues {Curve {keys [<at 0.5 0>,<at 1 0.4>]}}}, Scale Curve=6)
        Bias=0
       )
       Z=&Sine(
        Normalized Angle=&Add_Float(
         A=&Multiply_Float(A=$Particles.NormalizedAge, B=-1)
         B=&Multiply_Float(A=$Emitter.LoopedAge, B=1)
        )
        Period=1
        Scale=&FloatFromCurve(FloatCurve=DataInterface() {propertyValues {Curve {keys [<at 0.5 0>,<at 1 0.4>]}}}, Scale Curve=1.5)
        Bias=0
       )
      )
      Destination Space=::Simulation
      Apply Scale=true
     )
    )) {isSetParameters=true, module=SetVariables_#3}
  ]
  rendererClasses [&NiagaraRibbonRendererProperties]
  renderers [
   {rendererClass=&NiagaraRibbonRendererProperties}
   {rendererClass=&NiagaraRibbonRendererProperties, rendererIndex=1}
  ]
 }
 NE_VC_Boost_Ribbon_Parent001 {
  emitterSpawnScript [/SetVariables(Emitter.Desaturation:NiagaraFloat=&SelectFloatFromArray(
     Direct Array Index=$System.ColorVariantID
     Float Selection Array=DataInterface() {propertyValues {FloatData [0.2,1,1,1,1,1,1,1,1,1,1,1,1,1]}}
    )) {isSetParameters=true, module=SetVariables_#2}]
  emitterUpdateScript [
   &EmitterState(Life Cycle Mode=::NewEnumerator1, Loop Duration=1, UseLoopDelay=false)
   &VC_FilteredSocketCount()
   &VC_RateQualityHandler(Epic Spawn Rate=30, High Spawn Rate=30, Low Spawn Rate=30, Use Custom Rate?=true)
   &VC_SpawnRibbon_v1(Number of Ribbons=$Emitter.PortCount, SpawnRate=$Emitter.Spawn Rate)
  ]
  particleSpawnScript [
   &VC_InitializeRibbon_v1(
    Custom Offset=true
    Enable Skeletal Position=true
    Lifetime=&MergeHelper_FloatSwitch(0=1, 1=1, 2=1, 3=?, Index=$System.GameModeID, Num=::NewEnumerator1)
    Local Vector Offset=&AddVectorToPosition(Position=<-60 0 0>, Vector=Linked() {linkedVariable {name=User.DTCarOffset, type=NiagaraPosition}})
    Offset Amount=1
    Ribbon Width=50
   )
   &AddVelocity(Velocity=&MergeHelper_VectorSwitch(0=<-1200 0 0>, 1=<-1000 0 0>, 2=<-1500 0 0>, 3=?, Index=$System.GameModeID, Num=::NewEnumerator1)) {module=AddVelocity001}
   /SetVariables() {enabled=false, isSetParameters=true}
  ]
  particleUpdateScript [
   &VC_PaintedVariantsOverride(
    DefaultColor=&ColorFromCurve(ColorCurve=DataInterface() {propertyValues {
       AlphaCurve {keys [<at 0 1 RCIM_Cubic>,<at 0.3828 1>,<at 0.5954 1>,<at 0.9983 1 RCIM_Cubic>]}
       BlueCurve {keys [<at 0 0.9645>,<at 0.167 0.9936>,<at 0.3563 2.5>,<at 0.8373 1.3319>]}
       GreenCurve {keys [<at 0 0.9489>,<at 0.167 0.9991>,<at 0.3563 0.5635>,<at 0.8373 0.5225>]}
       RedCurve {keys [<at 0 1>,<at 0.167 1>,<at 0.3563 2.1256>,<at 0.8373 5>]}
      }})
    R2 Black:LinearColor=<rgba 1.5 1.5 1.5 10>
    R2 BurntSienna:LinearColor=<rgba 1.5 0.7555 0 25>
    R2 Cobalt:LinearColor=<rgba 0 1.883 8 4>
    R2 Crimson:LinearColor=<rgba 7.5 0.5859 0.7733 8>
    R2 DefaultColor:LinearColor=<rgba 2.4201 15 1.9756 1>
    R2 ForestGreen:LinearColor=<rgba 0 4.5 0.2274 10>
    R2 Grey:LinearColor=<rgba 3 3 3 5>
    R2 Lime:LinearColor=<rgba 3.8484 5.2135 1.82 5>
    R2 Orange:LinearColor=<rgba 7.3438 2.0077 0.6225 5>
    R2 Pink:LinearColor=<rgba 5 2.8385 4.7163 4>
    R2 Purple:LinearColor=<rgba 3.529 0.4167 8 5>
    R2 Saffron:LinearColor=<rgba 15 11.5 0 5>
    R2 SkyBlue:LinearColor=<rgba 0 6.6539 10 5>
    R2 TitaniumWhite:LinearColor=<rgba 15 15 15 1>
    Use R2 Painted Overrides=true
   )
   &ParticleState()
   &InheritSourceMovement(ApplyForce=false, ApplyPosition=true, ApplyPositionFalloff=false)
   &SolveForcesAndVelocity(Clamp Velocity=false, Limit Acceleration=false)
   &VC_UpdateRibbon_v1(Enable Skeletal Position=true)
   /SetVariables(Particles.Position=&AddVectorToPosition(
     Position=$Particles.Position
     Vector=&TransformVector(
      Vector=&MakeVector(
       X=0
       Y=&Sine(
        Normalized Angle=&Add_Float(
         A=&Multiply_Float(A=$Particles.NormalizedAge, B=-1)
         B=&Multiply_Float(A=$Emitter.LoopedAge, B=1)
        )
        Period=1
        Scale=&FloatFromCurve(FloatCurve=DataInterface() {propertyValues {Curve {keys [<at 0.5 0>,<at 1 0.4>]}}}, Scale Curve=6)
        Bias=0
       )
       Z=&Sine(
        Normalized Angle=&Add_Float(
         A=&Multiply_Float(A=$Particles.NormalizedAge, B=-1)
         B=&Multiply_Float(A=$Emitter.LoopedAge, B=1)
        )
        Period=1
        Scale=&FloatFromCurve(FloatCurve=DataInterface() {propertyValues {Curve {keys [<at 0.5 0>,<at 1 0.4>]}}}, Scale Curve=1.5)
        Bias=0
       )
      )
      Destination Space=::Simulation
      Apply Scale=true
     )
    )) {isSetParameters=true, module=SetVariables_#3}
   &Color(Color=<rgba 2.4201 15 1.9756 1>)
  ]
  rendererClasses [&NiagaraRibbonRendererProperties]
  renderers [
   {rendererClass=&NiagaraRibbonRendererProperties}
   {rendererClass=&NiagaraRibbonRendererProperties, rendererIndex=1}
  ]
 }
 NE_VC_Boost_Sprite_Parent {
  emitterSpawnScript [/SetVariables(
    Emitter.Desaturation:NiagaraFloat=&SelectFloatFromArray(Direct Array Index=$System.ColorVariantID, Float Selection Array=DataInterface() {propertyValues {FloatData [0,1,1,1,1,1,1,1,1,1,1,1,1,1]}})
    Emitter.Gain:NiagaraFloat=&SelectFloatFromArray(Direct Array Index=$System.ColorVariantID, Float Selection Array=DataInterface() {propertyValues {FloatData [1,1,5,6,2,3,3,2,4,5,5,4,4,4]}})
    Emitter.NeedReflection:NiagaraFloat=&SelectFloatFromArray(
     Direct Array Index=$System.ColorVariantID
     Float Selection Array=DataInterface() {propertyValues {FloatData [0,0.5,0.5,0.5,0.4,0.5,0.5,0.6,0.6,0.6,0.5,0.5,0.6,0.6]}}
    )
   ) {isSetParameters=true}]
  emitterUpdateScript [
   &EmitterState()
   &VC_FilteredSocketCount()
   &SpawnRate(SpawnRate=24, Use Spawn Probability=true)
   &SpawnRate(SpawnGroup=1, SpawnRate=24, Use Spawn Probability=true) {module=SpawnRate001}
   &SpawnRate(
    SpawnGroup=2
    SpawnRate=&MergeHelper_FloatSwitch(0=8, 1=16, 2=16, 3=?, Index=$System.GameModeID, Num=::NewEnumerator1)
    Use Spawn Probability=false
   ) {module=SpawnRate002}
   &SpawnRate(
    SpawnGroup=3
    SpawnRate=&MergeHelper_FloatSwitch(0=6, 1=12, 2=12, 3=?, Index=$System.GameModeID, Num=::NewEnumerator1)
    Use Spawn Probability=false
   ) {module=SpawnRate003}
  ]
  particleSpawnScript [
   /SetVariables(Particles.Visibility_Tag:NiagaraInt32=&SpawnGroup()) {isSetParameters=true}
   &VC_PaintedVariantsOverride(
    Black=<rgba 0.035 0.035 0.035 1>
    BurntSienna=<rgba 0.0625 0.0151 0 1>
    Cobalt=<rgba 0.0127 0.0618 0.8125 1>
    Crimson=<rgba 0.6146 0.0384 0.054 1>
    DefaultColor=<rgba 0.8385 0.8385 0.8385 1>
    ForestGreen=<rgba 0 0.5 0.0253 1>
    Grey=<rgba 0.5365 0.5365 0.5365 1>
    Lime=<rgba 0.5799 0.8802 0.2796 1>
    Orange=<rgba 0.6724 0.1141 0.021 1>
    Pink=<rgba 0.5052 0.1474 0.3751 1>
    Purple=<rgba 0.2831 0.1195 0.552 1>
    R1 Black:LinearColor=&RandomRangeLinearColor(Minimum=<rgba 0.0469 0.0469 0.0469 15>, Maximum=<rgba 0.022 0.022 0.022 15>)
    R1 BurntSienna:LinearColor=&RandomRangeLinearColor(Minimum=<rgba 0.0625 0.0212 0 5>, Maximum=<rgba 0.125 0.0424 0 5>)
    R1 Cobalt:LinearColor=&RandomRangeLinearColor(Minimum=<rgba 0.0413 0.0568 0.8802 55>, Maximum=<rgba 0.0344 0.0474 0.7344 55>)
    R1 Crimson:LinearColor=&RandomRangeLinearColor(Minimum=<rgba 0.5521 0.0173 0.0317 1>, Maximum=<rgba 0.3125 0.0098 0.018 1>)
    R1 DefaultColor=<rgba 0.5815 0.7861 1 1>
    R1 ForestGreen:LinearColor=&RandomRangeLinearColor(Minimum=<rgba 0 0.3958 0.02 1>, Maximum=<rgba 0 0.2812 0.0142 1>)
    R1 Grey:LinearColor=<rgba 0.3385 0.3385 0.3385 1>
    R1 Lime:LinearColor=&RandomRangeLinearColor(Minimum=<rgba 0.2865 0.5729 0 2>, Maximum=<rgba 0.2318 0.4635 0 2>)
    R1 Orange:LinearColor=&RandomRangeLinearColor(Minimum=<rgba 0.8229 0.2567 0 1>, Maximum=<rgba 0.8148 0.2542 0 1>)
    R1 Pink:LinearColor=&RandomRangeLinearColor(Minimum=<rgba 0.7865 0.1761 0.5645 1>, Maximum=<rgba 0.6562 0.147 0.4711 1>)
    R1 Purple:LinearColor=&RandomRangeLinearColor(Minimum=<rgba 0.2585 0.0812 0.5156 1>, Maximum=<rgba 0.2122 0.0666 0.4233 1>)
    R1 Saffron:LinearColor=&RandomRangeLinearColor(Minimum=<rgba 0.8021 0.6335 0.0794 1>, Maximum=<rgba 0.6042 0.4771 0.0598 1>)
    R1 SkyBlue:LinearColor=&RandomRangeLinearColor(Minimum=<rgba 0.0505 0.554 0.8073 1>, Maximum=<rgba 0.0723 0.4089 0.5781 1>)
    R2 Black:LinearColor=<rgba 0.0469 0.0469 0.0469 1>
    R2 BurntSienna:LinearColor=<rgba 0.0677 0.0163 0 1>
    R2 Cobalt:LinearColor=<rgba 0.0127 0.0618 0.8125 1>
    R2 Crimson:LinearColor=<rgba 0.6146 0.0384 0.054 1>
    R2 Grey:LinearColor=<rgba 0.5677 0.5677 0.5677 1>
    R2 Lime:LinearColor=<rgba 0.5799 0.8802 0.2796 1>
    R2 Orange:LinearColor=<rgba 0.6724 0.1141 0.021 1>
    R2 Pink:LinearColor=<rgba 0.5052 0.1474 0.3751 1>
    R2 Purple:LinearColor=<rgba 0.2831 0.1195 0.552 1>
    R2 Saffron:LinearColor=<rgba 0.8148 0.6584 0.1441 1>
    R2 SkyBlue:LinearColor=<rgba 0.1455 0.6482 0.901 1>
    Saffron=<rgba 0.8148 0.6584 0.1441 1>
    SkyBlue=<rgba 0.1455 0.6482 0.901 1>
    Use R1 Painted Overrides=true
    Use R2 Painted Overrides=true
   )
   &InitializeParticle(
    Color=$StackContext.Color
    Lifetime=&MaskFloatBySpawnGroup(
     Masked Float=&MergeHelper_FloatSwitch(
      0=&RandomRangeFloat(Minimum=0.25, Maximum=0.8)
      1=&RandomRangeFloat(Minimum=0.1, Maximum=0.3)
      2=&RandomRangeFloat(Minimum=0.25, Maximum=0.3)
      3=?
      Index=$System.GameModeID
      Num=::NewEnumerator1
     )
     Passthrough Float=&MaskFloatBySpawnGroup(
      Masked Float=&MergeHelper_FloatSwitch(
       0=&RandomRangeFloat(Minimum=0.5, Maximum=1.1)
       1=&RandomRangeFloat(Minimum=0.2, Maximum=0.5)
       2=&RandomRangeFloat(Minimum=0.2, Maximum=0.3)
       3=?
       Index=$System.GameModeID
       Num=::NewEnumerator1
      )
      Passthrough Float=&MaskFloatBySpawnGroup(
       Masked Float=&MergeHelper_FloatSwitch(
        0=&RandomRangeFloat(Minimum=0.5, Maximum=1.1)
        1=&RandomRangeFloat(Minimum=0.2, Maximum=0.5)
        2=&RandomRangeFloat(Minimum=0.2, Maximum=0.3)
        3=?
        Index=$System.GameModeID
        Num=::NewEnumerator1
       )
       Passthrough Float=&MaskFloatBySpawnGroup(
        Masked Float=&MergeHelper_FloatSwitch(
         0=&RandomRangeFloat(Minimum=0.25, Maximum=0.85)
         1=&RandomRangeFloat(Minimum=0.1, Maximum=0.3)
         2=&RandomRangeFloat(Minimum=0.25, Maximum=0.3)
         3=?
         Index=$System.GameModeID
         Num=::NewEnumerator1
        )
        Passthrough Float=0
        Spawn Group Mask=3
       )
      )
      Spawn Group Mask=1
     )
     Spawn Group Mask=2
    )
    Mesh Scale=<3 2 2>
    Mesh Scale Mode=::NewEnumerator4
    Sprite Rotation Mode=::NewEnumerator2
    Sprite Size Mode=::NewEnumerator1
    Uniform Sprite Size Max=17
    Uniform Sprite Size Min=12
    UsePositionOffset=false
    Write Material Random=true
   )
   &PartitionParticles(Number Of Partitions=$Emitter.PortCount)
   &SkeletalMeshLocation(
    Bone / Socket Index=$Particles.PartitionParticles.Partition
    Bone / Socket Sampling Mode=::NewEnumerator5
    MaskSpawnGroup=false
    Mesh Sampling Type=::NewEnumerator6
    Sampled Position Offset=&AddVectorToPosition(
     Position=&MaskVectorBySpawnGroup(
      Masked Vector=<-105 0 0>
      Passthrough Vector=&MaskVectorBySpawnGroup(
       Masked Vector=<-100 0 0>
       Passthrough Vector=&MaskVectorBySpawnGroup(
        Masked Vector=<-100 0 0>
        Passthrough Vector=&MaskVectorBySpawnGroup(Masked Vector=<-106 0 0>, Passthrough Vector=<0 0 0>, Spawn Group Mask=3)
       )
       Spawn Group Mask=1
      )
      Spawn Group Mask=2
     )
     Vector=Linked() {linkedVariable {name=User.DTCarOffset, type=NiagaraPosition}}
    )
    Skeletal Mesh=$System.SkeletalMesh
   )
   &AddVelocity(Velocity=&MaskVectorBySpawnGroup(
     Masked Vector=&MergeHelper_VectorSwitch(0=<-250 0 0>, 1=<-200 0 0>, 2=<-100 0 0>, 3=?, Index=$System.GameModeID, Num=::NewEnumerator1)
     Passthrough Vector=&MaskVectorBySpawnGroup(
      Masked Vector=&MergeHelper_VectorSwitch(0=<-4500 0 0>, 1=<-600 0 0>, 2=<-100 0 0>, 3=?, Index=$System.GameModeID, Num=::NewEnumerator1)
      Passthrough Vector=&MaskVectorBySpawnGroup(
       Masked Vector=&MergeHelper_VectorSwitch(0=<-5600 0 0>, 1=<-600 0 0>, 2=<-100 0 0>, 3=?, Index=$System.GameModeID, Num=::NewEnumerator1)
       Passthrough Vector=<1 1 1>
      )
      Spawn Group Mask=1
     )
     Spawn Group Mask=2
    )) {enabled=false}
   &InitialMeshOrientation(
    Mesh Orientation Mode=::NewEnumerator1
    Rotation=&RandomRangeVector(Minimum=<0 0 0>, Maximum=<0 0 0>)
    Rotation Coordinate Space=::NewEnumerator2
    Use Rotation Vector=true
   )
   &MeshRotationForce(
    Coordinate Space:ENiagaraMeshTransforms=::NewEnumerator3
    Rotation=&MaskVectorBySpawnGroup(
     Masked Vector=&MergeHelper_VectorSwitch(
      0=&RandomRangeVector(Minimum=<-21 0 0>, Maximum=<21 0 0>)
      1=&RandomRangeVector(Minimum=<-15 0 0>, Maximum=<15 0 0>)
      2=&RandomRangeVector(Minimum=<-21 0 0>, Maximum=<21 0 0>)
      3=?
      Index=$System.GameModeID
      Num=::NewEnumerator1
     )
     Passthrough Vector=&MaskVectorBySpawnGroup(
      Masked Vector=&MergeHelper_VectorSwitch(
       0=&RandomRangeVector(Minimum=<-18 0 0>, Maximum=<18 0 0>)
       1=&RandomRangeVector(Minimum=<-10 0 0>, Maximum=<10 0 0>)
       2=&RandomRangeVector(Minimum=<-18 0 0>, Maximum=<18 0 0>)
       3=?
       Index=0
       Num=::NewEnumerator1
      )
      Passthrough Vector=<0 0 0>
      Spawn Group Mask=2
     )
     Spawn Group Mask=3
    )
   )
   &ApplyInitialForces()
  ]
  particleUpdateScript [
   &ParticleState()
   &ShapeLocation(
    Handle Radius:NiagaraFloat=&MaskFloatBySpawnGroup(
     Masked Float=15
     Passthrough Float=&MaskFloatBySpawnGroup(
      Masked Float=15
      Passthrough Float=&MaskFloatBySpawnGroup(
       Masked Float=0
       Passthrough Float=&MaskFloatBySpawnGroup(Masked Float=0, Passthrough Float=0, Spawn Group Mask=3)
       Spawn Group Mask=2
      )
      Spawn Group Mask=1
     )
    )
    Large Radius=&MaskFloatBySpawnGroup(
     Masked Float=35
     Passthrough Float=&MaskFloatBySpawnGroup(
      Masked Float=30
      Passthrough Float=&MaskFloatBySpawnGroup(
       Masked Float=0
       Passthrough Float=&MaskFloatBySpawnGroup(Masked Float=0, Passthrough Float=0, Spawn Group Mask=3)
       Spawn Group Mask=2
      )
      Spawn Group Mask=1
     )
    )
    MaskSpawnGroup=false
    Offset=&TransformVector(
     Vector=&Lerp_Vector(
      A=<0 0 0>
      B=&MaskVectorBySpawnGroup(
       Masked Vector=&MergeHelper_VectorSwitch(
        0=&RandomRangeVector(Minimum=<-400 0 0>, Maximum=<-450 0 0>)
        1=&RandomRangeVector(Minimum=<-100 0 0>, Maximum=<-150 0 0>)
        2=&RandomRangeVector(Minimum=<-100 0 0>, Maximum=<-150 0 0>)
        3=?
        Index=$System.GameModeID
        Num=::NewEnumerator1
       )
       Passthrough Vector=&MaskVectorBySpawnGroup(
        Masked Vector=&MergeHelper_VectorSwitch(
         0=&RandomRangeVector(Minimum=<-350 0 0>, Maximum=<-400 0 0>)
         1=&RandomRangeVector(Minimum=<-100 0 0>, Maximum=<-150 0 0>)
         2=&RandomRangeVector(Minimum=<-100 0 0>, Maximum=<-150 0 0>)
         3=?
         Index=$System.GameModeID
         Num=::NewEnumerator1
        )
        Passthrough Vector=&MaskVectorBySpawnGroup(
         Masked Vector=&MergeHelper_VectorSwitch(0=<-200 0 0>, 1=<-100 0 0>, 2=<-100 0 0>, 3=?, Index=$System.GameModeID, Num=::NewEnumerator1)
         Passthrough Vector=&MaskVectorBySpawnGroup(
          Masked Vector=&MergeHelper_VectorSwitch(0=<-200 0 0>, 1=<-100 0 0>, 2=<-100 0 0>, 3=?, Index=$System.GameModeID, Num=::NewEnumerator1)
          Passthrough Vector=<0 0 0>
          Spawn Group Mask=3
         )
         Spawn Group Mask=2
        )
        Spawn Group Mask=1
       )
      )
      Alpha=$Particles.NormalizedAge
     )
     Destination Space=::Simulation
     Apply Scale=true
    )
    Offset Mode=::NewEnumerator0
    Override Seed=false
    Rotation Angle:NiagaraFloat=&Multiply_Float(A=$Engine.Time, B=0.95)
    Rotation Mode=::NewEnumerator1
    Shape Origin=$Particles.Initial.Position
    Shape Primitive=::NewEnumerator3
    UseOwnerScale=true
   ) {enabled=false, module=ShapeLocation003}
   &ShapeLocation(
    Handle Radius:NiagaraFloat=&MaskFloatBySpawnGroup(
     Masked Float=15
     Passthrough Float=&MaskFloatBySpawnGroup(
      Masked Float=15
      Passthrough Float=&MaskFloatBySpawnGroup(
       Masked Float=0
       Passthrough Float=&MaskFloatBySpawnGroup(Masked Float=0, Passthrough Float=0, Spawn Group Mask=3)
       Spawn Group Mask=2
      )
      Spawn Group Mask=1
     )
    )
    Large Radius=&MaskFloatBySpawnGroup(
     Masked Float=35
     Passthrough Float=&MaskFloatBySpawnGroup(
      Masked Float=30
      Passthrough Float=&MaskFloatBySpawnGroup(
       Masked Float=0
       Passthrough Float=&MaskFloatBySpawnGroup(Masked Float=0, Passthrough Float=0, Spawn Group Mask=3)
       Spawn Group Mask=2
      )
      Spawn Group Mask=1
     )
    )
    MaskSpawnGroup=false
    Offset=&TransformVector(
     Vector=&Lerp_Vector(
      A=<0 0 0>
      B=&MaskVectorBySpawnGroup(
       Masked Vector=&MergeHelper_VectorSwitch(
        0=&RandomRangeVector(Minimum=<-400 0 0>, Maximum=<-450 0 0>)
        1=&RandomRangeVector(Minimum=<-100 0 0>, Maximum=<-150 0 0>)
        2=&RandomRangeVector(Minimum=<-100 0 0>, Maximum=<-150 0 0>)
        3=?
        Index=$System.GameModeID
        Num=::NewEnumerator1
       )
       Passthrough Vector=&MaskVectorBySpawnGroup(
        Masked Vector=&MergeHelper_VectorSwitch(
         0=&RandomRangeVector(Minimum=<-350 0 0>, Maximum=<-400 0 0>)
         1=&RandomRangeVector(Minimum=<-100 0 0>, Maximum=<-150 0 0>)
         2=&RandomRangeVector(Minimum=<-100 0 0>, Maximum=<-150 0 0>)
         3=?
         Index=$System.GameModeID
         Num=::NewEnumerator1
        )
        Passthrough Vector=&MaskVectorBySpawnGroup(
         Masked Vector=&MergeHelper_VectorSwitch(0=<-200 0 0>, 1=<-100 0 0>, 2=<-100 0 0>, 3=?, Index=$System.GameModeID, Num=::NewEnumerator1)
         Passthrough Vector=&MaskVectorBySpawnGroup(
          Masked Vector=&MergeHelper_VectorSwitch(0=<-200 0 0>, 1=<-100 0 0>, 2=<-100 0 0>, 3=?, Index=$System.GameModeID, Num=::NewEnumerator1)
          Passthrough Vector=<0 0 0>
          Spawn Group Mask=3
         )
         Spawn Group Mask=2
        )
        Spawn Group Mask=1
       )
      )
      Alpha=$Particles.NormalizedAge
     )
     Destination Space=::Simulation
     Apply Scale=true
    )
    Offset Mode=::NewEnumerator0
    Override Seed=false
    Rotation Angle:NiagaraFloat=&Multiply_Float(
     A=$Engine.Time
     B=&MaskFloat(Mask Source=1, Mask Parameter=$Particles.PartitionParticles.Partition, Mask=0, Masked Float=-0.95, Passthrough Float=0.95)
    )
    Rotation Mode=::NewEnumerator1
    Shape Origin=$Particles.Initial.Position
    Shape Primitive=::NewEnumerator3
    UseOwnerScale=true
   )
   &SpriteRotationRate(Rotation Rate=&RandomRangeFloat(Minimum=-1, Maximum=1))
   &ScaleSpriteSize(Uniform Curve Sprite Scale=DataInterface() {propertyValues {Curve {
       defaultValue=0
       keys [<at 0 2.5 RCIM_Cubic>, {arriveTangent=-2.2574, interpMode=RCIM_Cubic, leaveTangent=-2.2574, tangentMode=RCTM_User, time=0.87, value=0}]
      }}})
   /SetVariables(Particles.WavesWater:LinearColor=<rgba 0.5815 0.7861 1 1>) {isSetParameters=true}
   &InheritSourceMovement(
    Applied Position Scale=&Multiply_VectorByFloat(
     Vector=<1 1 1>
     Float=&MergeHelper_FloatSwitch(
      0=0
      1=&FloatFromCurve(FloatCurve=DataInterface() {propertyValues {Curve {defaultValue=1, keys [<at 0 0 RCIM_Cubic>,<at 0.7 5 RCIM_Cubic>]}}})
      2=&FloatFromCurve(FloatCurve=DataInterface() {propertyValues {Curve {defaultValue=1, keys [<at 0 0 RCIM_Cubic>,<at 0.7 5 RCIM_Cubic>]}}})
      3=?
      Index=$System.GameModeID
      Num=::NewEnumerator1
     )
    )
    ApplyForce=false
    ApplyPosition=true
    ApplyPositionFalloff=false
   )
   &SolveForcesAndVelocity(Clamp Velocity=false, Limit Acceleration=false)
  ]
  rendererClasses [&NiagaraSpriteRendererProperties, &NiagaraMeshRendererProperties]
  renderers [
   {rendererClass=&NiagaraSpriteRendererProperties}
   {rendererClass=&NiagaraMeshRendererProperties, rendererIndex=1}
   {rendererClass=&NiagaraMeshRendererProperties, rendererIndex=2}
   {rendererClass=&NiagaraSpriteRendererProperties, rendererIndex=3}
  ]
 }
}
summary {
 systemName=NS_Boost_Lake01
 emitters [
  {
   emitterName=NE_VC_Boost_Sprite_Parent
   rendererClasses [&NiagaraSpriteRendererProperties, &NiagaraMeshRendererProperties]
  }
  {emitterName=NE_VC_Boost_Ribbon_Parent, rendererClasses [&NiagaraRibbonRendererProperties]}
  {emitterName=NE_VC_Boost_Ribbon_Parent001, rendererClasses [&NiagaraRibbonRendererProperties]}
  {
   emitterName=Meshes-LocalSpace
   rendererClasses [&NiagaraSpriteRendererProperties, &NiagaraMeshRendererProperties]
  }
 ]
 userVariables [
  User.bIsPainted {type=NiagaraBool}
  User.ColorVariantID {type=NiagaraInt32}
  User.DTCarOffset {type=NiagaraPosition, description="", defaultValue {}}
  User.GameModeID {type=NiagaraInt32}
  User.offset {type=NiagaraPosition, description="", defaultValue {}}
  User.PaintedColor {type=LinearColor, description="", defaultValue {struct=LinearColor, value=<rgba 1 1 1 1>}}
  User.PaintedScalar {type=NiagaraFloat, defaultValue=1, description=""}
  User.PortCount {type=NiagaraInt32, defaultValue=2}
  User.WaveMaxScale {type=NiagaraFloat, defaultValue=1.5, description=""}
  User.WavePeriod {type=NiagaraFloat, defaultValue=0.5, description=""}
 ]
}
